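# Physics asset for dynamic gear on the male medium COD Squad Plaid cosmetic body: belt clips, a belt pouch and a back grapple rope.
# Three clips attach to the pelvis with tight limits, while two-segment pouch and rope chains hang from tiny master bodies on the pelvis and right clavicle.
# Capsules on the pelvis, upper spine and right clavicle, upper arm and thigh form a partial body proxy, joined by fully free angular constraints.
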PhysicsAsset /BRCosmetics/Characters/Player/Male/Medium/Bodies/M_MED_COD_Squad_Plaid/Meshes/M_MED_COD_Squad_Plaid_Dyn
bodies {
 clavicle_r [{shapeType=Capsule, length=5.3925, radius=9.3397, center=<-9.8507 0.4373 0.3487>, rotation=<rot 90 0 0>}]
 dyn_belt_clip_01 [{shapeType=Box, extentX=11.0625, extentY=7.5, extentZ=2.4375, center=<-0.1907 0 1.5>}]
 dyn_belt_clip_02 [{shapeType=Box, extentX=11.0625, extentY=7.5, extentZ=2.4375, center=<-0.1907 0 1.5>}]
 dyn_belt_clip_03 [{shapeType=Box, extentX=11.0625, extentY=7.5, extentZ=2.4375, center=<-0.1907 0 1.5>}]
 dyn_belt_pouch_01 [{shapeType=Box, extentX=6.25, extentY=12, extentZ=1.5, center=<3.4104 0 0>}]
 dyn_belt_pouch_02 [{shapeType=Box, extentX=6.25, extentY=12, extentZ=1.5, center=<3.4104 0 0>}]
 dyn_belt_pouch_MSTR [{shapeType=Capsule, length=0.1, radius=0.1}]
 dyn_grapple_rope_bk_01 [{shapeType=Capsule, length=1.8665, radius=0.7235, center=<1.1991 0.0001 0>, rotation=<rot 90 0 0>}]
 dyn_grapple_rope_bk_02 [{shapeType=Box, extentX=9.625, extentY=7.5, extentZ=2.125, center=<4.0426 -0.0359 -0.7791>, rotation=<rot -10 0 2.5945>}]
 dyn_grapple_rope_bk_MSTR [{shapeType=Capsule, length=0.1, radius=0.1}]
 pelvis [{shapeType=Capsule, length=7.5175, radius=10.1209}]
 spine_03 [{shapeType=Capsule, length=7.1425, radius=12.2147}]
 spine_04 [
  {shapeType=Capsule, length=5.7363, radius=13.9959}
  {shapeType=Capsule, length=15.5, radius=4.8125, center=<4.4479 13.4214 -0.5568>, rotation=<rot -40.4605 -14.1933 -6.9646>}
 ]
 spine_05 [{shapeType=Capsule, length=21.33, radius=11.4647, center=<0 -0.4161 0>}]
 thigh_r [{shapeType=Capsule, length=13.4238, radius=7.6522, center=<10.6448 -0.7216 -0.0696>, rotation=<rot 90 0 0>}]
 upperarm_r [{shapeType=Capsule, length=22.705, radius=4.4647, center=<-9.8507 0.0001 0>, rotation=<rot 90 0 0>}]
}
constraints [
 {bone1Name=dyn_grapple_rope_bk_01, bone2Name=dyn_grapple_rope_bk_MSTR, swing1LimitDegrees=30, swing2LimitDegrees=40, twistLimitDegrees=5}
 {bone1Name=dyn_grapple_rope_bk_02, bone2Name=dyn_grapple_rope_bk_01, swing1LimitDegrees=20, swing2LimitDegrees=20, twistLimitDegrees=5}
 {bone1Name=dyn_belt_pouch_02, bone2Name=dyn_belt_pouch_01, swing1LimitDegrees=8, swing2LimitDegrees=45, twistLimitDegrees=5}
 {
  bone1Name=dyn_grapple_rope_bk_MSTR
  bone2Name=clavicle_r
  swing1LimitDegrees=45
  swing1Motion=Free
  swing2LimitDegrees=45
  swing2Motion=Free
  twistLimitDegrees=45
  twistMotion=Free
 }
 {bone1Name=dyn_belt_pouch_01, bone2Name=dyn_belt_pouch_MSTR, swing1LimitDegrees=8, swing2LimitDegrees=45, twistLimitDegrees=5}
 {
  bone1Name=spine_05
  bone2Name=spine_04
  swing1LimitDegrees=45
  swing1Motion=Free
  swing2LimitDegrees=45
  swing2Motion=Free
  twistLimitDegrees=45
  twistMotion=Free
 }
 {
  bone1Name=spine_04
  bone2Name=spine_03
  swing1LimitDegrees=45
  swing1Motion=Free
  swing2LimitDegrees=45
  swing2Motion=Free
  twistLimitDegrees=45
  twistMotion=Free
 }
 {
  bone1Name=spine_03
  bone2Name=pelvis
  swing1LimitDegrees=45
  swing1Motion=Free
  swing2LimitDegrees=45
  swing2Motion=Free
  twistLimitDegrees=45
  twistMotion=Free
 }
 {
  bone1Name=clavicle_r
  bone2Name=spine_05
  swing1LimitDegrees=45
  swing1Motion=Free
  swing2LimitDegrees=45
  swing2Motion=Free
  twistLimitDegrees=45
  twistMotion=Free
 }
 {
  bone1Name=thigh_r
  bone2Name=pelvis
  swing1LimitDegrees=45
  swing1Motion=Free
  swing2LimitDegrees=45
  swing2Motion=Free
  twistLimitDegrees=45
  twistMotion=Free
 }
 {bone1Name=dyn_belt_clip_01, bone2Name=pelvis, swing1LimitDegrees=10, swing2LimitDegrees=10, twistLimitDegrees=10}
 {bone1Name=dyn_belt_clip_02, bone2Name=pelvis, swing1LimitDegrees=10, swing2LimitDegrees=10, twistLimitDegrees=10}
 {bone1Name=dyn_belt_clip_03, bone2Name=pelvis, swing1LimitDegrees=10, swing2LimitDegrees=10, twistLimitDegrees=10}
 {
  bone1Name=upperarm_r
  bone2Name=clavicle_r
  swing1LimitDegrees=45
  swing1Motion=Free
  swing2LimitDegrees=45
  swing2Motion=Free
  twistLimitDegrees=45
  twistMotion=Free
 }
 {
  bone1Name=dyn_belt_pouch_MSTR
  bone2Name=pelvis
  swing1LimitDegrees=45
  swing1Motion=Free
  swing2LimitDegrees=45
  swing2Motion=Free
  twistLimitDegrees=45
  twistMotion=Free
 }
]
