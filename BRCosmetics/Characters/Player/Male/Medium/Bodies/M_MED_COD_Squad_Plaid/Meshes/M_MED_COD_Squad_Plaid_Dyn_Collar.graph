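# Physics asset for the dynamic collar on the male medium M_MED_COD_Squad_Plaid player body cosmetic.
# Five box bodies cover the left, right, back-left, back-right and back collar bones, each constrained to a tiny capsule on its matching MSTR bone with wide angular limits.
# Capsules on head and neck_02 add collision volumes beside the collar, joined by a constraint with free swing and twist.
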PhysicsAsset /BRCosmetics/Characters/Player/Male/Medium/Bodies/M_MED_COD_Squad_Plaid/Meshes/M_MED_COD_Squad_Plaid_Dyn_Collar
bodies {
 dyn_collar_back [{shapeType=Box, extentX=14.2169, extentY=20.6698, extentZ=1.2319, center=<7.0169 0 0.3071>, rotation=<rot 5.891 0 0>}]
 dyn_collar_back_MSTR [{shapeType=Capsule, length=0.5, radius=0.5, rotation=<rot 90 0 0>}]
 dyn_collar_bk_l [{shapeType=Box, extentX=11.8, extentY=19, extentZ=1.2319, center=<3.75 0 0>}]
 dyn_collar_bk_l_MSTR [{shapeType=Capsule, length=0.5, radius=0.5, rotation=<rot 90 0 0>}]
 dyn_collar_bk_r [{shapeType=Box, extentX=11.8, extentY=19, extentZ=1.2319, center=<3.75 0 0>}]
 dyn_collar_bk_r_MSTR [{shapeType=Capsule, length=0.5, radius=0.5, rotation=<rot 90 0 0>}]
 dyn_collar_l [{shapeType=Box, extentX=12.6726, extentY=11.7336, extentZ=1.2319, center=<3.002 0 0>}]
 dyn_collar_l_MSTR [{shapeType=Capsule, length=0.5, radius=0.5, rotation=<rot 90 0 0>}]
 dyn_collar_r [{shapeType=Box, extentX=12.6726, extentY=11.7336, extentZ=1.2319, center=<3.002 0 0>}]
 dyn_collar_r_MSTR [{shapeType=Capsule, length=0.5, radius=0.5, rotation=<rot 90 0 0>}]
 head [{shapeType=Capsule, length=9.6901, radius=6.9248, shapeName=headMain, center=<4.4458 2.1927 0>, rotation=<rot 90 0 0>}]
 neck_02 [{shapeType=Capsule, length=6.505, radius=4.255, shapeName=neck_02_capsule, center=<-0.9642 0 0>, rotation=<rot -0.0001 -90 90.0004>}]
}
constraints [
 {bone1Name=dyn_collar_r, bone2Name=dyn_collar_r_MSTR, swing1LimitDegrees=75, swing2LimitDegrees=75, twistLimitDegrees=75}
 {bone1Name=dyn_collar_bk_r, bone2Name=dyn_collar_bk_r_MSTR, swing1LimitDegrees=75, swing2LimitDegrees=75, twistLimitDegrees=75}
 {bone1Name=dyn_collar_back, bone2Name=dyn_collar_back_MSTR, swing1LimitDegrees=90, swing2LimitDegrees=60, twistLimitDegrees=90}
 {bone1Name=dyn_collar_bk_l, bone2Name=dyn_collar_bk_l_MSTR, swing1LimitDegrees=75, swing2LimitDegrees=75, twistLimitDegrees=75}
 {bone1Name=dyn_collar_l, bone2Name=dyn_collar_l_MSTR, swing1LimitDegrees=75, swing2LimitDegrees=75, twistLimitDegrees=75}
 {
  bone1Name=head
  bone2Name=neck_02
  swing1LimitDegrees=45
  swing1Motion=Free
  swing2LimitDegrees=45
  swing2Motion=Free
  twistLimitDegrees=45
  twistMotion=Free
 }
]
