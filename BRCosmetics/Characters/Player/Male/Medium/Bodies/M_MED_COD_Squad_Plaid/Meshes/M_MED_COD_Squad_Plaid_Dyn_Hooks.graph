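# Physics asset for the dangling grapple rope and hook on the COD Squad Plaid male medium body cosmetic.
# Three capsule rope links chain from a tiny master body on the right clavicle, with swing limits tightening toward the tip and a sphere bulking out the last link.
# Upper spine, right clavicle and upper arm capsules approximate the torso and arm around the rope, joined by constraints with free swing and twist.
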PhysicsAsset /BRCosmetics/Characters/Player/Male/Medium/Bodies/M_MED_COD_Squad_Plaid/Meshes/M_MED_COD_Squad_Plaid_Dyn_Hooks
bodies {
 clavicle_r [{shapeType=Capsule, length=5.3925, radius=9.3397, center=<-9.8507 0.4373 0.3487>, rotation=<rot 90 0 0>}]
 dyn_grapple_rope_fr_01 [{shapeType=Capsule, length=5.0866, radius=0.9378, center=<2.9605 0 0>, rotation=<rot 90 0 0>}]
 dyn_grapple_rope_fr_02 [{shapeType=Capsule, length=5.0866, radius=0.9378, center=<2.9605 0 0>, rotation=<rot 90 0 0>}]
 dyn_grapple_rope_fr_03 [
  {shapeType=Sphere, radius=5.0938, center=<11.4231 -0.0003 -1.3484>}
  {shapeType=Capsule, length=7.1804, radius=0.9378, center=<3.8985 0 -0.4333>, rotation=<rot 84.9999 0 0>}
 ]
 dyn_grapple_rope_fr_MSTR [{shapeType=Capsule, length=0.1, radius=0.1}]
 spine_03 [{shapeType=Capsule, length=7.1425, radius=12.2147}]
 spine_04 [
  {shapeType=Capsule, length=5.7363, radius=13.9959}
  {shapeType=Capsule, length=15.5, radius=4.8125, center=<4.4479 13.4214 -0.5568>, rotation=<rot -40.4605 -14.1933 -6.9646>}
 ]
 spine_05 [{shapeType=Capsule, length=21.33, radius=11.4647, center=<0 -0.4161 0>}]
 upperarm_r [{shapeType=Capsule, length=22.705, radius=4.4647, center=<-9.8507 0.0001 0>, rotation=<rot 90 0 0>}]
}
constraints [
 {bone1Name=dyn_grapple_rope_fr_01, bone2Name=dyn_grapple_rope_fr_MSTR, swing1LimitDegrees=40, swing2LimitDegrees=40, twistLimitDegrees=10}
 {bone1Name=dyn_grapple_rope_fr_02, bone2Name=dyn_grapple_rope_fr_01, swing1LimitDegrees=30, swing2LimitDegrees=30, twistLimitDegrees=10}
 {bone1Name=dyn_grapple_rope_fr_03, bone2Name=dyn_grapple_rope_fr_02, swing1LimitDegrees=20, swing2LimitDegrees=20, twistLimitDegrees=5}
 {
  bone1Name=dyn_grapple_rope_fr_MSTR
  bone2Name=clavicle_r
  swing1LimitDegrees=45
  swing1Motion=Free
  swing2LimitDegrees=45
  swing2Motion=Free
  twistLimitDegrees=45
  twistMotion=Free
 }
 {
  bone1Name=spine_05
  bone2Name=spine_04
  swing1LimitDegrees=45
  swing1Motion=Free
  swing2LimitDegrees=45
  swing2Motion=Free
  twistLimitDegrees=45
  twistMotion=Free
 }
 {
  bone1Name=spine_04
  bone2Name=spine_03
  swing1LimitDegrees=45
  swing1Motion=Free
  swing2LimitDegrees=45
  swing2Motion=Free
  twistLimitDegrees=45
  twistMotion=Free
 }
 {
  bone1Name=clavicle_r
  bone2Name=spine_05
  swing1LimitDegrees=45
  swing1Motion=Free
  swing2LimitDegrees=45
  swing2Motion=Free
  twistLimitDegrees=45
  twistMotion=Free
 }
 {
  bone1Name=upperarm_r
  bone2Name=clavicle_r
  swing1LimitDegrees=45
  swing1Motion=Free
  swing2LimitDegrees=45
  swing2Motion=Free
  twistLimitDegrees=45
  twistMotion=Free
 }
]
